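# Niagara melee trail effect for the Kitty Warrior Light Male one-handed pickaxe, built on the shared AnimTrail modules that sample a skeletal mesh's bones or sockets.
# AirTrail spawns a short-lived translucent ribbon that narrows and fades over life with a pale blue-white tint, feeding randomized values into dynamic material parameters.
# A Sparkle sprite emitter with curve-gated spawning, inherited velocity and a brighter blue tint is present but disabled.
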NiagaraSystem /BRCosmetics/Weapons/FORT_Melee/Pickaxe_Kitty_Warrior_Light_Male_1H/FX/NS_Pickaxe_KittyWarriorLightMale_AnimTrail
dependencies {
 usedDataInterfaces [&NiagaraDataInterfaceCurve, &NiagaraDataInterfaceColorCurve, &NiagaraDataInterfaceCurlNoise]
 usedDynamicInputs []
 usedModules [&SystemState]
}
emitters {
 AirTrail {
  emitterUpdateScript [
   &EmitterState(Life Cycle Mode=::NewEnumerator1, Loop Behavior=::NewEnumerator1, Loop Duration=0.3, UseLoopDelay=false)
   &FN_SpawnAnimTrail_v2(Spawn Interval=0)
  ]
  particleSpawnScript [
   /Game/Effects/Niagara/Modules/AnimTrail/FN_InitializeAnimTrail/FN_InitializeAnimTrailBoneIndecies(First Bone or Socket Index=1)
   /Game/Effects/Niagara/Modules/AnimTrail/FN_InitializeAnimTrail/FN_InitializeAnimTrail(AnimTrailWidth=0.75, Color=<rgba 1 1 1 0.4>, Lifetime=0.125)
   /SetVariables() {isSetParameters=true, module=SetVariables_#1}
  ]
  particleUpdateScript [
   &SolveForcesAndVelocity(Clamp Velocity=false, Limit Acceleration=false)
   &ScaleRibbonWidth(Ribbon Width Scale=&FloatFromCurve(FloatCurve=DataInterface() {propertyValues {Curve {keys [
         {arriveTangent=-1.2579, interpMode=RCIM_Cubic, time=0, value=1}
         {arriveTangent=-0.9664, interpMode=RCIM_Cubic, leaveTangent=-0.9664, tangentMode=RCTM_User, time=1, value=0.2}
        ]}}}))
   &FN_UpdateAnimTrail_v2()
   &DynamicMaterialParameters(
    Index 0 Param 1=&UniformRangedFloat(Minimum=0, Maximum=1)
    Index 0 Param 2=&UniformRangedFloat(Minimum=0, Maximum=1)
    Index 0 Param 3=1
    Index 0 Param 4=1
    Module.Param0WriteEnabled=true
    Module.Param1WriteEnabled=true
    Module.Param2WriteEnabled=true
    Module.Param3WriteEnabled=true
   )
   &Color(Color=&ColorFromCurve(ColorCurve=DataInterface() {propertyValues {
       AlphaCurve {keys [<at 0 0.5 RCIM_Cubic>,<at 1 0 RCIM_Cubic>]}
       BlueCurve {keys [<at 0.0114 1>]}
       GreenCurve {keys [<at 0.0114 0.9553>]}
       RedCurve {keys [<at 0.0114 0.7969>]}
      }}))
  ]
  rendererClasses [&NiagaraRibbonRendererProperties]
  renderers [{rendererClass=&NiagaraRibbonRendererProperties}]
 }
 Sparkle {
  bEnabled=false
  emitterUpdateScript [
   &EmitterState(Life Cycle Mode=::NewEnumerator1, Loop Behavior=::NewEnumerator1, Loop Duration=0.2, UseLoopDelay=false)
   &FN_SpawnAnimTrail_v2(
    AnimTrailWidthMode=::FromFirst
    Number of Anim Trails:NiagaraInt32=&Multiply_Int(
     A=1
     B=&MakeIntFromFloat(Float=&FloatFromCurve(
       FloatCurve=DataInterface() {propertyValues {Curve {keys [<at 0.45 0>,<at 0.46 1>,<at 0.59 1>,<at 0.65 1>,<at 0.66 0>]}}}
       CurveIndex=$Emitter.NormalizedLoopAge
       Scale Curve=1.1
      ))
    )
   )
   &SpawnPerUnit(
    Module.Use Max Movement Threshold=true
    Module.Use Movement Tolerance=false
    Spawn Per Unit:NiagaraFloat=&FloatFromCurve(
     FloatCurve=DataInterface() {propertyValues {Curve {keys [<at 0.2 100>,<at 0.3 3>,<at 0.4 3>,<at 0.5 200>]}}}
     CurveIndex=$Emitter.NormalizedLoopAge
     Scale Curve=2
    )
    Use Spawn Probability=false
   ) {enabled=false}
  ]
  particleSpawnScript [
   &FN_InitializeAnimTrailBoneIndecies(First Bone or Socket Index=1)
   &InitializeParticle(
    Sprite Rotation Mode=::NewEnumerator2
    Sprite Size Mode=::NewEnumerator1
    Uniform Sprite Size Max=16
    Uniform Sprite Size Min=10
    UsePositionOffset=false
    Write Material Random=true
   )
   &FN_InitializeAnimTrail(
    AnimTrailWidth=0.6
    Color=<rgba 0.7899 0.8465 1 0.7>
    Lifetime:NiagaraFloat=&RandomRangeFloat(Minimum=0.15, Maximum=0.2)
   )
   /SetVariables() {isSetParameters=true, module=SetVariables_#1}
   &InheritVelocity(Module.Limit Speed=true, Velocity Limit:NiagaraFloat=&RandomRangeFloat(Minimum=-20, Maximum=100))
  ]
  particleUpdateScript [
   /Niagara/Modules/Update/Forces/CurlNoiseForce(Noise Strength=500, Pan Noise=false) {enabled=false}
   &SolveForcesAndVelocity(Clamp Velocity=false, Limit Acceleration=false)
   &BoxLocation(Box Size=<80 5 5>, MaskSpawnGroup=false, Module.Override Seed=false, Rotate Box=false, Surface Only=false)
   &FN_UpdateAnimTrail_v2()
   &Color(Color=&ColorFromCurve(ColorCurve=DataInterface() {propertyValues {
       AlphaCurve {keys [
         {arriveTangent=-3.5673, interpMode=RCIM_Cubic, leaveTangent=-3.5673, time=0.7591, value=0.5}
         {arriveTangent=-3.5673, interpMode=RCIM_Cubic, leaveTangent=-3.5673, time=1, value=0}
        ]}
       BlueCurve {keys [<at 0.4339 4 RCIM_Cubic>]}
       GreenCurve {keys [<at 0.4339 3.5301 RCIM_Cubic>]}
       RedCurve {keys [<at 0.4339 2.6736 RCIM_Cubic>]}
      }}))
   &ParticleState()
   &ScaleSpriteSize(Scale Factor=&Vector2DFromFloat(Value=&FloatFromCurve(FloatCurve=DataInterface() {propertyValues {Curve {keys [
          {arriveTangent=4.4129, interpMode=RCIM_Cubic, leaveTangent=4.4129, tangentMode=RCTM_User, time=0, value=0.5}
          {arriveTangent=-4.1154, interpMode=RCIM_Cubic, leaveTangent=-4.1154, tangentMode=RCTM_User, time=0.2, value=1}
          {arriveTangent=-1.25, interpMode=RCIM_Cubic, leaveTangent=-1.25, time=1, value=0}
         ]}}}))) {enabled=false}
  ]
  rendererClasses [&NiagaraSpriteRendererProperties]
  renderers [{rendererClass=&NiagaraSpriteRendererProperties}]
 }
}
summary {
 systemName=NS_Pickaxe_KittyWarriorLightMale_AnimTrail
 emitters [
  {emitterName=AirTrail, rendererClasses [&NiagaraRibbonRendererProperties]}
  {bEnabled=false, emitterName=Sparkle, rendererClasses [&NiagaraSpriteRendererProperties]}
 ]
 userVariables [
  User.LifetimeCustom {type=NiagaraFloat, defaultValue=0, description=""}
  User.Primary Color {type=LinearColor, description="", defaultValue {struct=LinearColor, value=<rgba 1 1 1 1>}}
  User.RibbonIDOffset {type=NiagaraInt32}
  User.SkeletalMesh {
   type=NiagaraDataInterfaceSkeletalMesh
   defaultValue {
    struct=NiagaraExt_VariableValue_DataInterface
    value {dataInterface=&self:NiagaraDataInterfaceSkeletalMesh_0}
   }
  }
 ]
}
